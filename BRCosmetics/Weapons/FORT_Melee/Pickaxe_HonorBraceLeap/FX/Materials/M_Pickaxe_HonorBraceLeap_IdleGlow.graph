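# Translucent unlit glow material for Niagara mesh particles, named for the idle effect of the HonorBraceLeap pickaxe.
# Opacity comes from a reflected U gradient centered by view angle, sharpened by a falloff exponent, faded at the V ends and depth-faded.
# World position offset math widens the spline-like strip perpendicular to the view, scaled by a Fresnel-style term, and includes a push toward the camera.
# A color parameter feeds the shared emissive adjustments function.
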Material /BRCosmetics/Weapons/FORT_Melee/Pickaxe_HonorBraceLeap/FX/Materials/M_Pickaxe_HonorBraceLeap_IdleGlow
graph {
 properties {
  bAutomaticallySetUsageInEditor=false
  bUsedWithNiagaraMeshParticles=true
  blendMode=BLEND_Translucent
  shadingModel=MSM_Unlit
  usedShadingModels=Unlit
 }
 nodes [
  n1 = TextureCoordinate()
  n2 = Constant3Vector() {constant=<rgba 0 -1 0 0>}
  n3 = Subtract(A=CameraPositionWS(), B=WorldPosition().XYZ)
  n4 = ConstantBiasScale(None=DotProduct(A=Transform(None=@n2) {transformSourceType=TRANSFORMSOURCE_Instance}, B=Normalize(VectorInput=@n3)))
  n5 = Power(
   Base=MaterialFunctionCall!(
    Gradient=ComponentMask(None=@n1) {r=true}
    Center Point=NamedRerouteDeclaration(None=@n4) {name=View Angle Dot, nodeColor=<rgba 0 0.8125 1 1>}
   ) {materialFunction=/Game/Effects/Fort_Effects/MaterialFunctions/MF_GradientReflected}.Result
   Exp=ScalarParameter() {defaultValue=6, parameterName=Falloff Exponent}
  )
  n6 = TextureCoordinate()
  n7 = Abs(None=ConstantBiasScale(None=ComponentMask(None=@n6) {g=true}) {bias=-0.5, scale=2})
  n8 = Saturate(None=Divide(A=OneMinus(None=@n7), B=ScalarParameter() {defaultValue=0.1, parameterName=Spline Ends Fade Size}))
  DepthFade(
   Opacity=Divide(A=Multiply(A=@n5, B=@n8), B=NamedRerouteUsage())
   FadeDistance=ScalarParameter() {defaultValue=2, parameterName=Depth Fade Distance}
  )
  MaterialFunctionCall(Emissive=VectorParameter() {parameterName=Color, defaultValue=<rgba 2 0.5 0 1>}.RGB) {materialFunction=&MF_FX_EmissiveAdjustments}
  n9 = WorldPosition()
  n10 = NamedRerouteUsage()
  n11 = Normalize(VectorInput=Subtract(A=CameraPositionWS(), B=Add(A=@n9.XYZ, B=@n10)))
  n12 = ScalarParameter() {defaultValue=3, parameterName=Camera Offset}
  NamedRerouteDeclaration(None=Add(A=Multiply(A=@n11, B=@n12), B=NamedRerouteUsage())) {name=WPO, nodeColor=<rgba 1 0 0.5391 1>}
  n13 = Normalize(VectorInput=Subtract(A=CameraPositionWS(), B=WorldPosition().XYZ))
  n14 = CrossProduct(A=@n13, B=Transform(None=Constant3Vector() {constant=<rgba 0 1 0 0>}))
  n15 = Subtract(A=ComponentMask(None=TextureCoordinate()) {r=true}, B=NamedRerouteUsage()) {constB=0.5}
  n16 = Constant3Vector() {constant=<rgba 1 0 0 0>}
  n17 = Abs(None=DotProduct(A=@n13, B=Transform(None=@n16)))
  n18 = ScalarParameter() {defaultValue=1, parameterName=Fresnel Exponent}
  n19 = Multiply(
   A=NamedRerouteDeclaration(None=Power(Base=@n17, Exp=@n18) {constExponent=4}) {name=Fresnel, nodeColor=<rgba 0.6875 0 1 1>}
   B=ScalarParameter() {defaultValue=3, parameterName=Width}
  )
  NamedRerouteDeclaration(None=Multiply(A=Normalize(VectorInput=@n14), B=Multiply(A=@n15, B=@n19))) {name=SplineThicken WPO, nodeColor=<rgba 0.2891 0 1 1>}
  NamedRerouteUsage()
 ]
}
